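# Sound cue for a bus being destroyed, layering a vehicle destruction sound with bus horn honks in the Fort_Impacts class.
# A mixer blends a slightly lowered, reused police car destruction wave with multiple honks delayed by half a second, each through its own modulator.
# The root attenuation node overrides its settings with spatialization off, while the cue references medium-object demolition attenuation.
# A concurrency override prevents new instances once the limit is reached.
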SoundCue /Game/Sounds/Fort_Vehicles/Generic_Auto/Bus_Destroyed_02_Cue
root=Attenuation_0
nodes [
 {id=Attenuation_0, children [Mixer_0], properties {bOverrideAttenuation=true, attenuationOverrides {bSpatialize=false}}}
 {id=Delay_0, children [Modulator_0], properties {delayMax=0.5, delayMin=0.5}}
 {id=Mixer_0, children [Modulator_1,Delay_0], properties {inputVolume [0.8,1]}}
 {id=Modulator_0, children [WavePlayer_1]}
 {id=Modulator_1, children [WavePlayer_0]}
 {id=WavePlayer_0, properties {soundWaveAssetPtr=/Game/Sounds/Fort_Vehicles/Police/Police_Car_Destroyed_Stereo_01}}
 {id=WavePlayer_1, properties {soundWaveAssetPtr=/Game/Sounds/Fort_Vehicles/Honks_Alarms/Bus_Horn_Honk_Multiple_Honks}}
]
properties {
 attenuationSettings=/Game/Sounds/Attenuation/Fort_Demolition_Destroy_Medium_Sized_Object_Att
 bOverrideConcurrency=true
 maxDistance=4000
 soundClassObject=&Fort_Impacts
 volumeMultiplier=1.04
 concurrencyOverrides {resolutionRule=PreventNew}
}
