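# Sound cue for the second generic car alarm variant on Fortnite vehicles.
# A single wave player for Auto_Alarm_02 runs through a default modulator at reduced volume, routed to the impacts sound class.
# It uses medium-object demolition hit attenuation.
# Global and local alarm concurrency sets limit stacking, alongside a single-instance override that prevents new plays.
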SoundCue /Game/Sounds/Fort_Vehicles/Generic_Auto/Generic_Auto_Alarm_02_Cue
root=Modulator_0
nodes [
 {id=Modulator_0, children [WavePlayer_1]}
 {id=WavePlayer_1, properties {soundWaveAssetPtr=&/Auto_Alarm_02}}
]
properties {
 attenuationSettings=/Game/Sounds/Attenuation/Fort_Demolition_Hit_Medium_Sized_Object_Att
 maxDistance=4000
 soundClassObject=&Fort_Impacts
 volumeMultiplier=0.64
 concurrencyOverrides {maxCount=1, resolutionRule=PreventNew}
 concurrencySet [
  &/CONC_Generic_Auto_Alarm_Global
  &/CONC_Generic_Auto_Alarm_Local
 ]
}
