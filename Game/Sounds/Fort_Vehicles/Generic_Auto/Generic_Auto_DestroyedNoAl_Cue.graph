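# Sound cue for the destruction of a generic automobile, reusing a Police car destroyed stereo wave.
# A single wave player feeds a default Modulator for pitch and volume variation, with no random selection or layering.
# It plays at reduced volume in the Fort_Impacts class, using medium-sized demolition attenuation.
# An owner-only, single-instance prevent-new concurrency rule keeps repeat triggers from stacking.
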SoundCue /Game/Sounds/Fort_Vehicles/Generic_Auto/Generic_Auto_DestroyedNoAl_Cue
root=Modulator_1
nodes [
 {id=Modulator_1, children [WavePlayer_0]}
 {id=WavePlayer_0, properties {soundWaveAssetPtr=/Game/Sounds/Fort_Vehicles/Police/Police_Car_Destroyed_Stereo_01}}
]
properties {
 attenuationSettings=/Game/Sounds/Attenuation/Fort_Demolition_Destroy_Medium_Sized_Object_Att
 maxDistance=4000
 soundClassObject=&Fort_Impacts
 volumeMultiplier=0.56
 concurrencyOverrides {resolutionRule=PreventNew}
 concurrencySet [/Game/Sounds/Concurrency_Rules/Sound_Concurrency_SinglePreventNew_OwnerOnly]
}
