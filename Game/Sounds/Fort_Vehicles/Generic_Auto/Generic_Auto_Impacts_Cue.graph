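# Sound cue for generic car impact sounds, reusing stereo police car impact waves.
# A quality-level switch sends its first two slots to a modulated, equal-weight random pick of three waves, and its last two to a modulated single wave.
# It routes through the impacts sound class with medium-object hit attenuation and a shared object-impact concurrency set that limits overlapping hits.
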SoundCue /Game/Sounds/Fort_Vehicles/Generic_Auto/Generic_Auto_Impacts_Cue
root=QualityLevel_0
nodes [
 {id=Modulator_0, children [WavePlayer_0]}
 {id=Modulator_1, children [Random_0]}
 {id=QualityLevel_0, children [Modulator_1,Modulator_1,Modulator_0,Modulator_0]}
 {id=Random_0, children [WavePlayer_1,WavePlayer_2,WavePlayer_3], properties {weights [1,1,1]}}
 {id=WavePlayer_0, properties {soundWaveAssetPtr=/Game/Sounds/Fort_Vehicles/Police/Police_Car_Impact_Stereo_01}}
 {id=WavePlayer_1, properties {soundWaveAssetPtr=&Police_Car_Impact_Stereo_01}}
 {id=WavePlayer_2, properties {soundWaveAssetPtr=/Game/Sounds/Fort_Vehicles/Police/Police_Car_Impact_Stereo_02}}
 {id=WavePlayer_3, properties {soundWaveAssetPtr=/Game/Sounds/Fort_Vehicles/Police/Police_Car_Impact_Stereo_03}}
]
properties {
 attenuationSettings=/Game/Sounds/Attenuation/Fort_Demolition_Hit_Medium_Sized_Object_Att
 maxDistance=4000
 soundClassObject=&Fort_Impacts
 volumeMultiplier=0.6
 concurrencyOverrides {maxCount=3, resolutionRule=StopOldest, volumeScale=0.8}
 concurrencySet [&CONC_Object_Impacts_FourMax]
}
